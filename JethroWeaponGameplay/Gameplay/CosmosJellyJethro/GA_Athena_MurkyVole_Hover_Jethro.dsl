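; Gameplay ability for a hover move that decides whether it may activate around the Force Jump power.
; CanActivateAbility checks the owner's ability system tags: with Force Jump granted and the Galileo Lobster item equipped, hover activates only while Force Jump is on cooldown.
; Without the Force Jump tag it allows activation.
; No hover behavior is implemented here, and the end-ability event holds only a disabled parent call.
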(event Ability|EventOnEndAbility
  (params (bWasCancelled :type Boolean))
  (|CallParentFunction bWasCancelled) ; This node is disabled and will not be called.\nDrag off pins to build functionality.
)

(fn CanActivateAbility
  (params (ActorInfo :type Gameplay%20Ability%20Actor%20Info%20Structure) (Handle :type Gameplay%20Ability%20Spec%20Handle%20Structure))
  (bind _owneractor (Utilities|Struct|BreakGameplayAbilityActorInfo ActorInfo))
  (bind _returnvalue (Ability|GetAbilitySystemComponent _owneractor))
  (bind _haslightsabertags (LocalVariables|GetHasLightsaberTags))
  (if (GameplayTags|HasMatchingGameplayTag _returnvalue "(TagName=\"Granted.Athena.PlotTwist.Force.Power.Jump\")")
    (if (GameplayTags|HasMatchingGameplayTag _returnvalue "(TagName=\"Granted.Athena.Galileo.Lobster.Equipped\")")
      (if (GameplayTags|HasMatchingGameplayTag _returnvalue "(TagName=\"Granted.Cooldown.Athena.PlotTwist.Force.Jump\")")
        (return _haslightsabertags true)
        (else
          (return _haslightsabertags false)))
      (else
        _)))
  (return _haslightsabertags true))
