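; Gameplay ability for the Jethro item, active while in inventory, that shows or hides the Jump and GamepadJump HUD action keys for hover activation.
; If the hotfix flag allows it, the keys start visible and two WaitGameplayTagQuery tasks on the hover-activatable query toggle them; otherwise the ability ends immediately.
; Both tasks are stored so ability end can stop them and hide the keys.
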(event Ability|EventActivateAbility
  (bind _hoveractivatablequery (Variables|Default|GetHoverActivatableQuery))
  ;[0 1 0] Activate Ability - Activated when in inventory
  (|CallParentFunction)
  (Default|UpdateActivateHUDActionKey false) ; Make sure it is not visible
  (if (ScalableFloat|GetValueasBool (Variables|Hotfix|GetShowHoverActivateKey))
    (Default|UpdateActivateHUDActionKey true) ; Initially visible
    (bind _asynctaskproxy (Ability|Tasks|WaitGameplayTagQuery _hoveractivatablequery)
      (:then
        (Variables|Default|SetShowHoverTrueTask _asynctaskproxy)
        (bind _asynctaskproxy_2 (Ability|Tasks|WaitGameplayTagQuery _hoveractivatablequery 0 "WhenFalse")
          (:then
            (Variables|Default|SetShowHoverFalseTask _asynctaskproxy_2))
          (:Triggered
            (Default|UpdateActivateHUDActionKey false))))
      (:Triggered
        (Default|UpdateActivateHUDActionKey true)))
    (else
      ;;[1 0.25 0] End Ability
      (Ability|EndAbility))))

(event Ability|EventOnEndAbility
  (params (bWasCancelled :type Boolean))
  ;[1 0.25 0] End Ability - Hide HUD Action keys and end async task
  (|CallParentFunction bWasCancelled)
  (Utilities|FlowControl|Sequence
    (:then_0
      (bind _showhovertruetask (Variables|Default|GetShowHoverTrueTask)
        (:then
          (GameplayTasks|EndTask _showhovertruetask))
        (:else)))
    (:then_1
      (bind _showhoverfalsetask (Variables|Default|GetShowHoverFalseTask)
        (:then
          (GameplayTasks|EndTask _showhoverfalsetask))
        (:else)))
    (:then_2
      (Default|UpdateActivateHUDActionKey false))))

(fn UpdateActivateHUDActionKey
  (params (IsVisible :type Boolean))
  (CallFunction|SetActionDisplayVisibility "Jump" IsVisible)
  (CallFunction|SetActionDisplayVisibility "GamepadJump" IsVisible)
  (CallFunction|SetActionDisplayActive IsVisible))
