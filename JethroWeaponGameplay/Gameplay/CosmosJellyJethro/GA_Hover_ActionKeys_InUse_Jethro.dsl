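; Passive gameplay ability for the Jethro item that shows Ascend, Boost and Cancel HUD action keys while the player hovers.
; It activates when the item is added to inventory, hides the keys, then watches the hovering state tag count to toggle them.
; Each key's visibility is also gated by its own hotfixable scalable float.
; Ending the ability hides the keys and ends the tag-watch task.
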(event Ability|EventActivateAbility
  ; Passive Ability activated when item is added to inventory -
  (|CallParentFunction)
  (CallFunction|UpdateAllHUDActionKeys false)
  (bind (_asynctaskproxy _tagcount) (Ability|Tasks|WaitGameplayTagCountChange (Variables|Default|GetInHoveringStateTag))
    (:then
      (Variables|Default|SetInHoverStateTask _asynctaskproxy))
    (:TagCountChanged
      (CallFunction|UpdateAllHUDActionKeys (>= _tagcount 1)))))

(event Ability|EventOnEndAbility
  (params (bWasCancelled :type Boolean))
  ;[1 0.25 0] Cleanup on endability
  (|CallParentFunction bWasCancelled)
  (CallFunction|UpdateAllHUDActionKeys false)
  (bind _inhoverstatetask (Variables|Default|GetInHoverStateTask)
    (:then
      (GameplayTasks|EndTask _inhoverstatetask))
    (:else)))

(fn UpdateAllHUDActionKeys
  (params (Visible :type Boolean))
  (bind _visible (Variables|GetVisible))
  (CallFunction|SetActionDisplayActive Visible)
  (Utilities|FlowControl|Sequence
    (:then_0
      (CallFunction|UpdateAscendHUDActionKey _visible))
    (:then_1
      (CallFunction|UpdateBoostHUDActionKey _visible))
    (:then_2
      (CallFunction|UpdateCancelHUDActionKey _visible))))

(fn UpdateAscendHUDActionKey
  (params (IsVisible :type Boolean))
  (bind _output_get (LocalVariables|SetShouldBeVisible (and IsVisible (ScalableFloat|GetValueasBool (Variables|Hotfix|GetShowAscendKeyWhileHovering)))))
  (CallFunction|SetActionDisplayVisibility "HoverAscend" _output_get)
  (CallFunction|SetActionDisplayVisibility "HoverAscend" _output_get))

(fn UpdateBoostHUDActionKey
  (params (IsVisible :type Boolean))
  (bind _output_get (LocalVariables|SetShouldBeVisible (and IsVisible (ScalableFloat|GetValueasBool (Variables|Hotfix|GetShowBoostKeyWhileHovering)))))
  (CallFunction|SetActionDisplayVisibility "HoverBoost" _output_get)
  (CallFunction|SetActionDisplayVisibility "HoverBoost" _output_get))

(fn UpdateCancelHUDActionKey
  (params (IsVisible :type Boolean))
  (bind _output_get (LocalVariables|SetShouldBeVisible (and IsVisible (ScalableFloat|GetValueasBool (Variables|Hotfix|GetShowCancelKeyWhileHovering)))))
  (CallFunction|SetActionDisplayVisibility "HoverCancel" _output_get)
  (CallFunction|SetActionDisplayVisibility "HoverCancel" _output_get))
