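; Gameplay ability for a Jethro hover weapon that manages fuel regeneration.
; On activation it runs the parent logic and gets the owner pawn.
; It then watches the owner's reserve fuel attribute with a threshold task and stops fuel regen when that value drops to zero or below.
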(event Ability|EventActivateAbility
  (|CallParentFunction)
  (bind _ownerpawn (Variables|Default|GetOwnerPawn)
    (:then
      ;[1 0.3352 0.1135] Stop fuel regen when reserve fuel is empty
      (bind (_asynctaskproxy _bmatchescomparison) (Ability|Tasks|WaitForAttributeChangeThreshold (Variables|Default|GetReserveFuelAttribute) "LessThanOrEqualTo" 0.0 false _ownerpawn)
        (:OnChange
          (if _bmatchescomparison
            (CallFunction|StopFuelRegen)))))
    (:else)))
