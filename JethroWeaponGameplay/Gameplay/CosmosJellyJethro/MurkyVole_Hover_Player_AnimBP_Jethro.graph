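# Player AnimBP for a MurkyVole jetpack hover, providing Item Override and MovementMode animation layers.
# In MovementMode_FinalPoseOverride, a Movement state machine (rise-up, leaning base, boost, post-boost) feeds a Twist machine that applies bone-masked additive twist by targeting state.
# That result, switchable with a jump-apex pose, is layered with the incoming pose from Spine_03.
# ItemUpperBody defines a hover arms-additive state machine; BeginPlay raises IsStartBlend one tick late.
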AnimBlueprint /JethroWeaponGameplay/Gameplay/CosmosJellyJethro/MurkyVole_Hover_Player_AnimBP_Jethro
target_skeleton=&Fortnite_M_Avg_Player_Skeleton
graphs [
 AnimGraph {nodes [Output Pose()]}
 EventGraph {nodes [
   Event BlueprintBeginPlay() {is=title K2Node_Event}
   > Set IsStartBlend(IsStartBlend:bool=false) -> (Output_Get:bool)
   > DelayUntilNextTick!(execute) -> (then)
   > Set IsStartBlend(IsStartBlend=true)
  ]}
 ItemAimOffset {nodes [
   Input Pose!() -> (Pose, AimOffsetAlpha:double, Yaw:double, Pitch:double) {name=InPose_AimOffset}
   Input Pose!() -> (Pose) {name=InPose_PreAimOffsets}
   Output Pose() {layerGroup=Item Override}
  ]}
 ItemFallAdditive {nodes [
   Input Pose() {name=InPoseItemFallAdditive}
   Output Pose() {layerGroup=Item Override}
  ]}
 ItemFallLandAdditive {nodes [
   Input Pose() {name=InPoseItemFallLandAdditive}
   Output Pose() {layerGroup=Item Override}
  ]}
 ItemFallLevel2Override {nodes [
   Input Pose() {name=InPose_FallLevel2}
   Output Pose() {layerGroup=Item Override}
  ]}
 ItemFinalPoseOverride {nodes [
   Input Pose() {name=InFinalPose}
   Output Pose() {layerGroup=Item Override}
  ]}
 ItemFullBodyAimOffsets {nodes [
   Input Pose!() -> (Pose, FullBody_AimOffsetAlpha:double, Aim_Yaw:double, Aim_Pitch:double) {name=InPoseFullBodyAimOffsets}
   Input Pose!() -> (Pose) {name=InPosePreFullBodyAimOffsets}
   Output Pose() {layerGroup=Item Override}
  ]}
 ItemFullBodyOverride {nodes [
   Input Pose() {name=InPoseFullBody}
   Output Pose() {layerGroup=Item Override}
  ]}
 ItemIdleOverride {nodes [
   Input Pose() {name=InPose_Idle}
   Output Pose() {layerGroup=Item Override}
  ]}
 ItemJetPackJumpAdditive {nodes [Output Pose() {layerGroup=Item Override}]}
 ItemJetPackStartAdditive {nodes [Output Pose() {layerGroup=Item Override}]}
 ItemJumpLoopAdditive {nodes [
   Input Pose() {name=InPoseItemJumpLoopAdditive}
   Output Pose() {layerGroup=Item Override}
  ]}
 ItemJumpOverride {nodes [
   Input Pose() {name=InPoseJump}
   Output Pose() {layerGroup=Item Override}
  ]}
 ItemJumpUpAdditive {nodes [
   Input Pose() {name=InPoseItemJumpUpAdditive}
   Output Pose() {layerGroup=Item Override}
  ]}
 ItemPreAimOffsets {nodes [
   Input Pose() {name=InPosePreAimOffsets}
   Output Pose() {layerGroup=Item Override}
  ]}
 ItemPreIK {nodes [
   Input Pose() {name=InPose_PreIK}
   Output Pose() {layerGroup=Item Override}
  ]}
 ItemRightHandIK {nodes [
   Input Pose() {name=InPosePreRightHandIK}
   Input Pose() {name=InPosePostRightHandIK}
   Output Pose() {layerGroup=Item Override}
  ]}
 ItemSkeletalControl {nodes [
   Input Pose()
   Output Pose() {layerGroup=Item Override}
  ]}
 ItemSwimJumpFallLoop {nodes [Output Pose() {layerGroup=Item Override}]}
 ItemSwimJumpSurfaceEnd {nodes [Output Pose() {layerGroup=Item Override}]}
 ItemSwimJumpSurfaceLoop {nodes [Output Pose() {layerGroup=Item Override}]}
 ItemUpperBody {nodes [
   Input Pose!() -> (Pose, InputParam:Fort Anim Input Adjusted Aim Structure) {name=InPoseUpperBody}
   Output Pose() {layerGroup=Item Override}
   Save cached pose 'UpperBody Pose'() {cacheName=UpperBody Pose}
   AnimGraphNode_StateMachine_2 = AdditivePose SM() {is=title AnimGraphNode_StateMachine}
  ]}
 AdditivePose SM {
  entry=PassThrough
  path=ItemUpperBody.AnimGraphNode_StateMachine_2.AdditivePose SM
  ids=id
  nodes [
   To AddArms() {is=title AnimStateAliasNode, also {aliases [PassThrough]}}
   AnimStateNode_0 = PassThrough() {is=title AnimStateNode}
   AnimStateNode_3 = AddArms() {is=title AnimStateNode}
   AnimStateTransitionNode_2 = AnimStateTransitionNode()
   AnimStateTransitionNode_4 = AnimStateTransitionNode()
  ]
  links [
   {
    dst=PassThrough
    rule=ItemUpperBody.AnimGraphNode_StateMachine_2.AdditivePose SM.AnimStateTransitionNode_4.Transition
    src=AddArms
    via=AnimStateTransitionNode_4
   }
   {
    dst=AddArms
    rule=ItemUpperBody.AnimGraphNode_StateMachine_2.AdditivePose SM.AnimStateTransitionNode_2.AnimationTransitionGraph_1
    src=To AddArms
    via=AnimStateTransitionNode_2
   }
  ]
 }
 PassThrough {
  path=ItemUpperBody.AnimGraphNode_StateMachine_2.AdditivePose SM.AnimStateNode_0.PassThrough
  nodes [Output Animation Pose(Result=Use cached pose 'UpperBody Pose'())]
 }
 AddArms {
  path=ItemUpperBody.AnimGraphNode_StateMachine_2.AdditivePose SM.AnimStateNode_3.AddArms
  nodes [Output Animation Pose(Result=Apply Additive(
     Base=Use cached pose 'UpperBody Pose'()
     Additive="Sequence Player (additive) 'JumpApex_NoWep__Loop_CMM_Additive'"() {sequence=/Game/Animation/Game/MainPlayer/Locomotion/Medium/Male/Jump/NoWep/JumpApex_NoWep__Loop_CMM_Additive}
     Alpha=0.5
    ) {alphaBoolBlend {blendOption=Linear}})]
 }
 AnimationTransitionGraph_1 {
  path=ItemUpperBody.AnimGraphNode_StateMachine_2.AdditivePose SM.AnimStateTransitionNode_2.AnimationTransitionGraph_1
  nodes [Result(bCanEnterTransition=Get bIsHoverAddArms() -> (bIsHoverAddArms:bool))]
 }
 Transition {
  path=ItemUpperBody.AnimGraphNode_StateMachine_2.AdditivePose SM.AnimStateTransitionNode_4.Transition
  nodes [Result(bCanEnterTransition=NOT Boolean(A=Get bIsHoverAddArms()))]
 }
 ItemZipLineStartAdditive {nodes [Output Pose() {layerGroup=Item Override}]}
 MovementMode_AimOffset {nodes [
   Input Pose!() -> (Pose) {name=InAimOffsetsPose}
   Input Pose() {name=InSourcePosePreAimOffsets}
   Output Pose() {layerGroup=MovementMode}
  ]}
 MovementMode_BaseLocomotion {nodes [
   Input Pose() {name=InPoseBaseLocomotion}
   Output Pose() {layerGroup=MovementMode}
  ]}
 MovementMode_FinalPoseOverride {nodes [
   n1 = Sequence Player 'M_JumpApex_ARDS_Loop'() {bLoopAnimation=false, sequence=/Game/Animation/Game/MainPlayer/Locomotion/Medium/Male/Jump/AssaultRifle/M_JumpApex_ARDS_Loop}
   AnimGraphNode_StateMachine_1 = Twist SM() {is=title AnimGraphNode_StateMachine}
   Output Pose(Result=Layered blend per bone(
     BasePose=Blend Poses by bool(BlendPose_0=@AnimGraphNode_StateMachine_1, BlendPose_1=@n1, BlendTime_0=0.5) {blendType=Linear}
     BlendPoses_0=Use cached pose 'Input Pose'()
    ) {bMeshSpaceRotationBlend=true, layerSetup [{branchFilters [{blendDepth=3, boneName=Spine_03}]}]}) {layerGroup=MovementMode}
   Save cached pose 'Input Pose'(Pose=Input Pose() {name=InFinalPose}) {cacheName=Input Pose}
   AnimGraphNode_StateMachine_0 = Movement() {is=title AnimGraphNode_StateMachine}
   Save cached pose 'PreTwist Pose'(Pose=@AnimGraphNode_StateMachine_0) {cacheName=PreTwist Pose}
  ]}
 Movement {
  entry=RiseUp
  path=MovementMode_FinalPoseOverride.AnimGraphNode_StateMachine_0.Movement
  ids=id
  nodes [
   AnimStateNode_0 = RiseUp() {is=title AnimStateNode}
   AnimStateNode_1 = Base() {
    is=title AnimStateNode
    stateEntered {guid=#1}
    stateFullyBlended {guid=#2}
    stateLeft {guid=#3}
   }
   AnimStateNode_2 = Boost() {
    is=title AnimStateNode
    stateEntered {guid=#4}
    stateFullyBlended {guid=#5}
    stateLeft {guid=#6}
   }
   AnimStateNode_3 = PostBoostTransition() {
    is=title AnimStateNode
    stateEntered {guid=#7}
    stateFullyBlended {guid=#8}
    stateLeft {guid=#9}
   }
   AnimStateTransitionNode_0 = AnimStateTransitionNode() {blendMode=QuadraticInOut, crossfadeDuration=0.25}
   AnimStateTransitionNode_1 = AnimStateTransitionNode() {
    blendMode=QuadraticInOut
    crossfadeDuration=0.3
    transitionEnd {guid=#10}
    transitionInterrupt {guid=#11}
    transitionStart {guid=#12}
   }
   AnimStateTransitionNode_2 = AnimStateTransitionNode() {
    bAutomaticRuleBasedOnSequencePlayerInState=true
    blendMode=QuadraticInOut
    crossfadeDuration=0.25
    transitionEnd {guid=#13}
    transitionInterrupt {guid=#14}
    transitionStart {guid=#15}
   }
   AnimStateTransitionNode_3 = AnimStateTransitionNode() {bAutomaticRuleBasedOnSequencePlayerInState=true, crossfadeDuration=1}
  ]
  links [
   {
    dst=Boost
    rule=MovementMode_FinalPoseOverride.AnimGraphNode_StateMachine_0.Movement.AnimStateTransitionNode_0.Transition
    src=Base
    via=AnimStateTransitionNode_0
   }
   {
    dst=PostBoostTransition
    rule=MovementMode_FinalPoseOverride.AnimGraphNode_StateMachine_0.Movement.AnimStateTransitionNode_1.Transition
    src=Boost
    via=AnimStateTransitionNode_1
   }
   {
    dst=Base
    rule=MovementMode_FinalPoseOverride.AnimGraphNode_StateMachine_0.Movement.AnimStateTransitionNode_2.Transition
    src=PostBoostTransition
    via=AnimStateTransitionNode_2
   }
   {
    dst=Base
    rule=MovementMode_FinalPoseOverride.AnimGraphNode_StateMachine_0.Movement.AnimStateTransitionNode_3.Transition
    src=RiseUp
    via=AnimStateTransitionNode_3
   }
  ]
 }
 RiseUp {
  path=MovementMode_FinalPoseOverride.AnimGraphNode_StateMachine_0.Movement.AnimStateNode_0.RiseUp
  nodes [Output Animation Pose(Result=Sequence Player 'MurkyVole_Pack_Jets_RiseUp_CMM'() {bLoopAnimation=false, sequence=/MurkyVoleGameplay/Assets/Backpack/Animations/Jets/Player/MurkyVole_Pack_Jets_RiseUp_CMM})]
 }
 Base {
  path=MovementMode_FinalPoseOverride.AnimGraphNode_StateMachine_0.Movement.AnimStateNode_1.Base
  nodes [Output Animation Pose(Result=Apply Additive(
     Base=Blendspace Player 'MurkyVole_Pack_Jets_Leans_CMM_BS'(X=0, Y=0) {blendSpace=/MurkyVoleGameplay/Assets/Backpack/Animations/Jets/Player/MurkyVole_Pack_Jets_Leans_CMM_BS}
     Additive="Sequence Player (additive) 'MurkyVole_Pack_Jets_Idle_CMM'"() {sequence=/MurkyVoleGameplay/Assets/Backpack/Animations/Jets/Player/MurkyVole_Pack_Jets_Idle_CMM}
    ) {alphaBoolBlend {blendOption=Linear}})]
 }
 Boost {
  path=MovementMode_FinalPoseOverride.AnimGraphNode_StateMachine_0.Movement.AnimStateNode_2.Boost
  nodes [Output Animation Pose(Result=Blendspace Player 'MurkyVole_Pack_Jets_Boost_CMM_BS'(X=0, Y=0) {blendSpace=/MurkyVoleGameplay/Assets/Backpack/Animations/Jets/Player/MurkyVole_Pack_Jets_Boost_CMM_BS})]
 }
 PostBoostTransition {
  path=MovementMode_FinalPoseOverride.AnimGraphNode_StateMachine_0.Movement.AnimStateNode_3.PostBoostTransition
  nodes [
   Sequence Player 'MurkyVole_Pack_Jets_Boost_Overlap_Left_CMM'() {bLoopAnimation=false, sequence=/MurkyVoleGameplay/Assets/Backpack/Animations/Jets/Player/MurkyVole_Pack_Jets_Boost_Overlap_Left_CMM}
   Output Animation Pose(Result=Blendspace Player 'MurkyVole_Pack_Jets_BoostOverlap_CMM_BS'(X=0, Y=0) {bLoop=false, blendSpace=/MurkyVoleGameplay/Assets/Backpack/Animations/Jets/Player/MurkyVole_Pack_Jets_BoostOverlap_CMM_BS})
  ]
 }
 Transition {
  path=MovementMode_FinalPoseOverride.AnimGraphNode_StateMachine_0.Movement.AnimStateTransitionNode_0.Transition
  nodes [Result(bCanEnterTransition=Get bIsHoverDash() -> (bIsHoverDash:bool))]
 }
 Transition {
  path=MovementMode_FinalPoseOverride.AnimGraphNode_StateMachine_0.Movement.AnimStateTransitionNode_1.Transition
  nodes [Result(bCanEnterTransition=Get bCanEnterPostBoost() -> (bCanEnterPostBoost:bool))]
 }
 Transition {path=MovementMode_FinalPoseOverride.AnimGraphNode_StateMachine_0.Movement.AnimStateTransitionNode_2.Transition, nodes [Result()]}
 Transition {path=MovementMode_FinalPoseOverride.AnimGraphNode_StateMachine_0.Movement.AnimStateTransitionNode_3.Transition, nodes [Result()]}
 Twist SM {
  entry=Non Targeted
  path=MovementMode_FinalPoseOverride.AnimGraphNode_StateMachine_1.Twist SM
  ids=id
  nodes [
   AnimStateNode_0 = Non Targeted() {is=title AnimStateNode}
   AnimStateNode_1 = Targeted() {is=title AnimStateNode}
   AnimStateTransitionNode_0 = AnimStateTransitionNode()
   AnimStateTransitionNode_1 = AnimStateTransitionNode()
  ]
  links [
   {
    dst=Targeted
    rule=MovementMode_FinalPoseOverride.AnimGraphNode_StateMachine_1.Twist SM.AnimStateTransitionNode_0.Transition
    src=Non Targeted
    via=AnimStateTransitionNode_0
   }
   {
    dst=Non Targeted
    rule=MovementMode_FinalPoseOverride.AnimGraphNode_StateMachine_1.Twist SM.AnimStateTransitionNode_1.Transition
    src=Targeted
    via=AnimStateTransitionNode_1
   }
  ]
 }
 Non Targeted {
  path=MovementMode_FinalPoseOverride.AnimGraphNode_StateMachine_1.Twist SM.AnimStateNode_0.Non Targeted
  nodes [
   n1 = Layered blend per bone(
    BasePose=Apply Additive(
     Base=Use cached pose 'PreTwist Pose'()
     Additive=Blendspace Player 'MurkyVole_Pack_Twist_BS'!(X=0) -> (Pose) {blendSpace=/MurkyVoleGameplay/Assets/Backpack/Animations/Jets/Player/MurkyVole_Pack_Twist_BS}
    ) {alphaBoolBlend {blendOption=Linear}}
    BlendPoses_0=Use cached pose 'PreTwist Pose'()
   ) {bMeshSpaceRotationBlend=true, layerSetup [{branchFilters [{blendDepth=2, boneName=Spine_04}, {blendDepth=-1, boneName=Neck_01}]}]}
   Output Animation Pose(Result=@n1)
  ]
 }
 Targeted {
  path=MovementMode_FinalPoseOverride.AnimGraphNode_StateMachine_1.Twist SM.AnimStateNode_1.Targeted
  nodes [
   n1 = Blendspace Player 'MurkyVole_Pack_Twist_BS'(X=0) {blendSpace=&MurkyVole_Pack_Twist_BS}
   n2 = Use cached pose 'PreTwist Pose'()
   Output Animation Pose(Result=Layered blend per bone(BasePose=Apply Additive(Base=@n2, Additive=@n1) {alphaBoolBlend {blendOption=Linear}}, BlendPoses_0=Use cached pose 'PreTwist Pose'()) {bMeshSpaceRotationBlend=true, layerSetup [{branchFilters [{blendDepth=2, boneName=Spine_04}]}]})
  ]
 }
 Transition {
  path=MovementMode_FinalPoseOverride.AnimGraphNode_StateMachine_1.Twist SM.AnimStateTransitionNode_0.Transition
  nodes [Result(bCanEnterTransition=Get bIsTargeting())]
 }
 Transition {
  path=MovementMode_FinalPoseOverride.AnimGraphNode_StateMachine_1.Twist SM.AnimStateTransitionNode_1.Transition
  nodes [Result(bCanEnterTransition=NOT Boolean(A=Get bIsTargeting()))]
 }
 MovementMode_FullBodyOverride {nodes [
   Input Pose() {name=InputPoseFullBody}
   Output Pose() {layerGroup=MovementMode}
  ]}
 MovementMode_InAirOverride {nodes [Output Pose(Result=Blend Poses by bool(BlendPose_0=Input Pose() {name=InSource_Pose}, BlendPose_1=Input Pose() {name=InAirPose}, BlendTime_0=0.3) {blendType=Linear}) {layerGroup=MovementMode}]}
 MovementMode_LocomotionAdditives {nodes [
   Input Pose() {name=InLocomotionAdditivesPose}
   Output Pose() {layerGroup=MovementMode}
  ]}
 MovementMode_LowerBodyOverride {nodes [
   Input Pose() {name=InLowerBodyPose}
   Input Pose() {name=InSourcePose}
   Output Pose() {layerGroup=MovementMode}
  ]}
 MovementMode_OffScreen {nodes [
   Input Pose() {name=InPoseOffscreen}
   Output Pose() {layerGroup=MovementMode}
  ]}
 MovementMode_SourcePose {nodes [
   Input Pose() {name=InSourcePoseBase}
   Output Pose() {layerGroup=MovementMode}
  ]}
]
variables [IsStartBlend {type=boolean}]
